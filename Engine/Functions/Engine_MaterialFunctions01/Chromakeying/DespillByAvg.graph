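# Chromakeying material function that despills an image color by limiting green to an average-based value.
# SplitComponents separates the input, then an Add and Divide derive that limit, with both Add inputs wired to red as written.
# An If replaces green with the limit only where green exceeds it.
# MakeFloat3 rebuilds the color from that green, with red wired to both the X and Z inputs.
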MaterialFunction /Engine/Functions/Engine_MaterialFunctions01/Chromakeying/DespillByAvg
graph {
 properties {description=Despill using average color of input, libraryCategoriesText [Chromakeying]}
 nodes [
  n1 = MaterialFunctionCall(Color=FunctionInput() {inputName=Image Color}) {materialFunction=&SplitComponents}
  n2 = Divide(A=Add(A=@n1.R, B=@n1.R))
  FunctionOutput(None=MaterialFunctionCall(X=@n1.R, Y=If(A=@n1.G, B=@n2, "A > B"=@n2, "A == B"=@n1.G, "A < B"=@n1.G), Z=@n1.R) {materialFunction=&MakeFloat3}.Result)
 ]
}
