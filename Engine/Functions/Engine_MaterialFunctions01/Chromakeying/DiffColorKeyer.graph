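# Material function that keys out a green screen with a color-difference method and outputs RGBA.
# A single Custom HLSL node measures green's excess over weighted red and blue, normalized by the key color's smallest, epsilon-clamped green difference.
# The result is scaled by a threshold, remapped by black/white clip and offset, then inverted into alpha.
# Despite its name, the UnPremult input blends RGB toward alpha-premultiplied color.
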MaterialFunction /Engine/Functions/Engine_MaterialFunctions01/Chromakeying/DiffColorKeyer
graph {nodes [
  MaterialExpressionFunctionInput_1 = FunctionInput() {bUsePreviewValueAsDefault=true, inputName=GreenScreen, previewValue=<1 0.5 0 1>}
  MaterialExpressionFunctionInput_0 = FunctionInput() {bUsePreviewValueAsDefault=true, inputName=KeyColor, sortPriority=1, previewValue=<0 1 0 1>}
  MaterialExpressionFunctionInput_5 = FunctionInput() {bUsePreviewValueAsDefault=true, inputName=AlphaThresOffset, inputType=FunctionInput_Vector2, sortPriority=2, previewValue=<1 0 0 1>}
  MaterialExpressionFunctionInput_2 = FunctionInput() {bUsePreviewValueAsDefault=true, inputName=WeightsRB, inputType=FunctionInput_Vector2, sortPriority=3, previewValue=<0.5 0.5 0 1>}
  MaterialExpressionFunctionInput_3 = FunctionInput() {bUsePreviewValueAsDefault=true, inputName=ClipBW, inputType=FunctionInput_Vector2, sortPriority=4, previewValue=<0 1 0 1>}
  MaterialExpressionFunctionInput_4 = FunctionInput() {bUsePreviewValueAsDefault=true, inputName=UnPremult, inputType=FunctionInput_Scalar, sortPriority=5}
  n1 = Custom!(
   GreenScreen=@MaterialExpressionFunctionInput_1
   KeyColor=@MaterialExpressionFunctionInput_0
   AlphaThresOffset=@MaterialExpressionFunctionInput_5
   WeightsRB=@MaterialExpressionFunctionInput_2
   ClipBW=@MaterialExpressionFunctionInput_3
   Unpremult=@MaterialExpressionFunctionInput_4
  ) {
   code="GetDiffColorKeyer(GreenScreen, KeyColor, AlphaThresOffset, WeightsRB, ClipBW, Unpremult)"
   description=ColorDiffKeyer
   outputType=CMOT_Float4
   preamble="float4 GetDiffColorKeyer(\r\n    float3 GreenScreen,\r\n    float3 KeyColor,\r\n    float2 AlphaThresOffset,\r\n    float2 WeightsRB,\r\n    float2 ClipBW,\r\n    float Unpremult\r\n)\r\n{\r\n    const float epsilon = 0.001;\r\n\r\n    float diffGR = max(KeyColor.g - KeyColor.r, epsilon); // max to avoid division by zero\r\n    float diffGB = max(KeyColor.g - KeyColor.b, epsilon);\r\n    float diff = min(diffGR, diffGB);\r\n\r\n    float weightedRB = GreenScreen.g - (GreenScreen.r * WeightsRB.x + GreenScreen.b * WeightsRB.y);\r\n    float alpha = weightedRB / diff;\r\n\r\n    // alpha post-processing\r\n    alpha /= AlphaThresOffset.x;\r\n\r\n    // remap to 0-1\r\n    alpha = (alpha - ClipBW.x) / (ClipBW.y - ClipBW.x);\r\n    alpha = saturate((alpha - AlphaThresOffset.y) / (1.0 - AlphaThresOffset.y));\r\n    alpha = 1.0 - alpha;\r\n\r\n    // premultiply\r\n    float3 rgb = lerp(GreenScreen, GreenScreen * alpha, Unpremult);\r\n\r\n    return float4(rgb, alpha);\r\n}"
   inputs [
    {
     inputName=GreenScreen
     input {expression=&self:MaterialExpressionFunctionInput_1}
    }
    {
     inputName=KeyColor
     input {expression=&self:MaterialExpressionFunctionInput_0}
    }
    {
     inputName=AlphaThresOffset
     input {expression=&self:MaterialExpressionFunctionInput_5}
    }
    {
     inputName=WeightsRB
     input {expression=&self:MaterialExpressionFunctionInput_2}
    }
    {
     inputName=ClipBW
     input {expression=&self:MaterialExpressionFunctionInput_3}
    }
    {
     inputName=Unpremult
     input {expression=&self:MaterialExpressionFunctionInput_4}
    }
   ]
  }
  FunctionOutput(None=@n1) {outputName=RGBA}
 ]}
