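# RGBtoHSV is an engine material function, exposed to the material library in the Chromakeying folder, that converts an RGB color to hue, saturation and value.
# A single input feeds a Custom HLSL node whose float3 result goes straight to the output.
# The code orders the channels with step and lerp instead of if statements, and adds a small epsilon to avoid division by zero.
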MaterialFunction /Engine/Functions/Engine_MaterialFunctions01/Chromakeying/RGBtoHSV
graph {
 properties {bExposeToLibrary=true}
 nodes [
  MaterialExpressionFunctionInput_0 = FunctionInput()
  n1 = Custom!(c=@MaterialExpressionFunctionInput_0) {
   code="float4 K = float4(0.0, -1.0 / 3.0, 2.0 / 3.0, -1.0);\r\nfloat4 p = lerp(float4(c.bg, K.wz), float4(c.gb, K.xy), step(c.b, c.g));\r\nfloat4 q = lerp(float4(p.xyw, c.r), float4(c.r, p.yzx), step(p.x, c.r)); \r\nfloat d = q.x - min(q.w, q.y);\r\nfloat e = .0000001;\r\nreturn float3(abs(q.z + (q.w - q.y) / (6.0 * d + e)), d / (q.x + e), q.x);"
   inputs [{inputName=c, input {expression=&self:MaterialExpressionFunctionInput_0}}]
  }
  FunctionOutput(None=@n1)
 ]
}
